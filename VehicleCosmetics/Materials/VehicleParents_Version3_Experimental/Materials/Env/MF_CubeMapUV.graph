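# MF_CubeMapUV is a material function that outputs a direction vector, named for cubemap lookup in vehicle environment materials.
# It subtracts an origin from world position, blending that origin between object pivot plus a Position offset and camera position via the enumerated Use Position parameter, which defaults to camera.
# A branch normalizes the vector and rotates it about Z by a degree parameter, but is unconnected, so the output stays unnormalized and unrotated.
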MaterialFunction /VehicleCosmetics/Materials/VehicleParents_Version3_Experimental/Materials/Env/MF_CubeMapUV
graph {nodes [
  n1 = Divide(A=ScalarParameter() {defaultValue=347, parameterName="Rotation (Degree)"}) {constB=360}
  n2 = MaterialFunctionCall() {materialFunction=&ObjectPivotPoint}
  n3 = VectorParameter() {parameterName=Position, defaultValue=<rgba 0 0 182 0>}
  n4 = ScalarParameter() {
   controlType=Enumeration
   defaultValue=1
   enumeration=&/UsePivot
   parameterName=Use Position
  }
  n5 = Subtract(A=WorldPosition().XYZ, B=LinearInterpolate(A=Add(A=@n2.Object Pivot Location, B=@n3.RGB), B=CameraPositionWS(), Alpha=@n4))
  n6 = Normalize(VectorInput=@n5)
  Add(
   A=RotateAboutAxis(NormalizedRotationAxis=Constant3Vector() {constant=<rgba 0 0 1 0>}, RotationAngle=Frac(None=@n1), PivotPoint=Constant(), Position=@n6)
   B=@n6
  )
  FunctionOutput(None=@n5)
 ]}
