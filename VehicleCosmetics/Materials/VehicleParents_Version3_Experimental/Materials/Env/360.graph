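# Experimental unlit environment material for vehicle cosmetics, usable on skeletal meshes, that samples the sky light environment map.
# A cube-map UV function supplies the sample direction, with scalar parameters for roughness and intensity.
# The scaled color's channels are summed into a grayscale value that feeds separate divide and power branches.
# Side nodes hold a day cubemap sample, reflection vector and normal blend; no material output connection is shown.
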Material /VehicleCosmetics/Materials/VehicleParents_Version3_Experimental/Materials/Env/360
graph {
 properties {bUsedWithSkeletalMesh=true, shadingModel=MSM_Unlit, usedShadingModels=Unlit}
 nodes [
  Constant() {r=0.5}
  n1 = MaterialFunctionCall() {materialFunction=&/MF_CubeMapUV}
  MaterialExpressionMultiply_0 = Multiply(
   A=SkyLightEnvMapSample(Direction=@n1.Result, Roughness=ScalarParameter() {defaultValue=-0.024, parameterName=Rough})
   B=ScalarParameter() {defaultValue=1, parameterName=M}
  )
  n2 = Convert(None=@MaterialExpressionMultiply_0) {
   nodeName=Break Float3
   convertInputs [{
     type=Vector3
     expressionInput {
      expression=&self:MaterialExpressionMultiply_0
      inputName=None
      mask=0
      maskA=0
      maskB=0
      maskG=0
      maskR=0
      outputIndex=0
     }
    }]
   convertOutputs [{type=Scalar}, {type=Scalar}, {type=Scalar}]
  }
  n3 = Add(A=Add(A=@n2, B=@n2), B=@n2)
  Divide(A=@n3, B=ScalarParameter() {defaultValue=28.8623, parameterName=Diver})
  LinearInterpolate(A=Constant3Vector(), B=PixelNormalWS(), Alpha=ScalarParameter() {parameterName=Ref})
  Power(Base=@n3, Exp=ScalarParameter() {defaultValue=4.497})
  ReflectionVectorWS(CustomWorldNormal=@n1.Result)
  TextureSample(UVs=@n1.Result) {
   samplerType=SAMPLERTYPE_LinearColor
   texture=&/T_CF_Day_cube_2k_v2
  }
 ]
}
