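# ChromeProbe is a minimal Default Lit material in the experimental Version 3 vehicle-cosmetics environment set, named as a chrome reflection probe.
# Its graph holds only three scalar constants, with no textures, parameters or math.
# The values fit a fully metallic, mirror-smooth white surface, though the graph does not show which material inputs they drive.
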Material /VehicleCosmetics/Materials/VehicleParents_Version3_Experimental/Materials/Env/ChromeProbe
graph {
 properties {usedShadingModels=Default Lit}
 nodes [
  Constant() {r=1}
  Constant() {r=1}
  Constant()
 ]
}
